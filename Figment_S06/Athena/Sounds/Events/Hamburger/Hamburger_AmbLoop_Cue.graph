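# Hamburger ambient loop Sound Cue for the Figment_S06 event sounds, currently an empty shell with no root or sound nodes, so it plays no audio.
# It is routed to the in-game SFX sound class and overrides attenuation with distance falloff disabled.
# The override still enables listener focus and trace-based occlusion, both interpolated, with non-focus and occluded volume scales set to zero.
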SoundCue /Figment_S06/Athena/Sounds/Events/Hamburger/Hamburger_AmbLoop_Cue
root=null
nodes []
properties {
 bOverrideAttenuation=true
 maxDistance=2097152
 soundClassObject=&Fort_SFX_InGame
 volumeMultiplier=1
 attenuationOverrides {
  bAttenuate=false
  bEnableFocusInterpolation=true
  bEnableListenerFocus=true
  bEnableOcclusion=true
  focusAttackInterpSpeed=0.1
  focusAzimuth=3
  focusReleaseInterpSpeed=0.5
  nonFocusAzimuth=7
  nonFocusVolumeAttenuation=0
  occlusionInterpolationTime=2
  occlusionTraceChannel=ECC_GameTraceChannel11
  occlusionVolumeAttenuation=0
  stereoSpread=100000
 }
}
