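# SoundCue for the map-wide Hamburger event ambience, mixing a main event layer with a proximity loop and a quiet, non-spatialized edge wind bed.
# A Progress parameter hard-switches three ambient loops at staged thresholds, with a randomly delayed, repeating banner sound during the first stage only.
# The main layer uses a shared focus attenuation asset.
# Proximity and wind use long-range custom attenuation and play only on the first two quality levels.
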SoundCue /Figment_S06/Athena/Sounds/Events/Hamburger/Hamburger_AmbLoop_Cue_3_2_1
root=Mixer_0
nodes [
 {
  id=Attenuation_0
  children [WavePlayer_5]
  properties {
   bOverrideAttenuation=true
   attenuationOverrides {
    bEnableFocusInterpolation=true
    bSpatialize=false
    distanceAlgorithm=Custom
    falloffDistance=400000
    focusVolumeAttenuation=1.2
    nonFocusVolumeAttenuation=0.7
    stereoSpread=100000
    customAttenuationCurve {editorCurveData {keys [
       {interpMode=RCIM_Cubic, tangentMode=RCTM_User, time=0, value=1}
       {interpMode=RCIM_Cubic, tangentMode=RCTM_User, time=0.99, value=1}
       <at 1 0>
      ]}}
   }
  }
 }
 {
  id=Attenuation_1
  children [Mixer_2]
  properties {attenuationSettings=&/Hamburger_AmbientLoop_Focus_Att}
 }
 {
  id=Attenuation_2
  children [WavePlayer_1]
  properties {
   bOverrideAttenuation=true
   attenuationOverrides {
    bEnableFocusInterpolation=true
    bEnableListenerFocus=true
    distanceAlgorithm=Custom
    falloffDistance=400000
    focusVolumeAttenuation=1.2
    nonFocusVolumeAttenuation=0.7
    stereoSpread=100000
    customAttenuationCurve {editorCurveData {keys [
       {interpMode=RCIM_Cubic, tangentMode=RCTM_User, time=0, value=1}
       {interpMode=RCIM_Cubic, tangentMode=RCTM_User, time=0.99, value=1}
       <at 1 0>
      ]}}
   }
  }
 }
 {id=Delay_0, children [WavePlayer_4], properties {delayMax=240, delayMin=30}}
 {id=Looping_0, children [Delay_0]}
 {id=Mixer_0, children [Attenuation_1, QualityLevel_0, QualityLevel_1], properties {inputVolume [1,0.5,0.05]}}
 {id=Mixer_2, children [ParamCrossFade_1, ParamCrossFade_0], properties {inputVolume [0.4,1]}}
 {
  id=ParamCrossFade_0
  children [WavePlayer_2,WavePlayer_3,WavePlayer_0]
  properties {
   paramName=Progress
   crossFadeInput [
    {fadeInDistanceEnd=0, fadeInDistanceStart=0, fadeOutDistanceEnd=0.7333, fadeOutDistanceStart=0.7333, volume=1}
    {fadeInDistanceEnd=0.7333, fadeInDistanceStart=0.7333, fadeOutDistanceEnd=0.8667, fadeOutDistanceStart=0.8667, volume=1}
    {fadeInDistanceEnd=0.8667, fadeInDistanceStart=0.8667, fadeOutDistanceEnd=100, fadeOutDistanceStart=100, volume=0.75}
   ]
  }
 }
 {
  id=ParamCrossFade_1
  children [Looping_0]
  properties {
   paramName=Progress
   crossFadeInput [{fadeInDistanceEnd=0, fadeInDistanceStart=0, fadeOutDistanceEnd=0.7333, fadeOutDistanceStart=0.7333, volume=1}]
  }
 }
 {id=QualityLevel_0, children [Attenuation_2, Attenuation_2, null, null]}
 {id=QualityLevel_1, children [Attenuation_0, Attenuation_0, null, null]}
 {id=WavePlayer_0, properties {bLooping=true, soundWaveAssetPtr=&/Hamburger_AmbientLoop_03}}
 {
  id=WavePlayer_1
  properties {bLooping=true, soundWaveAssetPtr=&/Hamburger_AmbientLoop_Proximity_01}
 }
 {id=WavePlayer_2, properties {bLooping=true, soundWaveAssetPtr=&/Hamburger_AmbientLoop_01}}
 {id=WavePlayer_3, properties {bLooping=true, soundWaveAssetPtr=&/Hamburger_AmbientLoop_02}}
 {id=WavePlayer_4, properties {soundWaveAssetPtr=&/Hamburger_Banner}}
 {id=WavePlayer_5, properties {bLooping=true, soundWaveAssetPtr=/Figment_S06/Sounds/Fort_Ambient_Looping/Edge/fort_amb_edge_wind_01}}
]
properties {
 maxDistance=400400
 soundClassObject=/Figment_S06/Athena/Sounds/Events/SoundClasses/Fort_Events_Mapwide
 volumeMultiplier=1
 attenuationOverrides {
  bEnableFocusInterpolation=true
  bEnableListenerFocus=true
  bEnableOcclusion=true
  falloffDistance=50000
  focusAttackInterpSpeed=0.1
  focusAzimuth=3
  focusReleaseInterpSpeed=0.5
  nonFocusAzimuth=7
  nonFocusVolumeAttenuation=0
  occlusionInterpolationTime=2
  occlusionTraceChannel=ECC_GameTraceChannel11
  occlusionVolumeAttenuation=0
  stereoSpread=100000
 }
}
